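# Animation Blueprint for the Googley glider parts, posing streamers and googly eyes on the player parachute.
# It copies the attached parent's pose, collapses both streamer root bones to zero scale, then runs world-space Anim Dynamics on each eye.
# A final RigidBody node blends in at up to 0.3 once the glider is fully deployed.
# Each update checks deployment, holding streamers collapsed until then and afterward easing their scale override out.
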AnimBlueprint /Game/Gadgets/Assets/VinderTech_GliderChute/Glider_Googley/Meshes/SK_Glider_Googley_Parts_AnimBP
target_skeleton=&/SK_Glider_Googley_Parts_S_GoogleyGlider_Parachute_Skeleton
graphs [
 AnimGraph {nodes [
   n1 = Get StreamersAlpha() -> (StreamersAlpha:double)
   n2 = "Transform (Modify) Bone - Bone: streamer_L_01"(
    ComponentPose=Local To Component(LocalPose=Copy Pose From Mesh() {bUseAttachedParent=true})
    Alpha=@n1
    Translation="0,0,0"
    Rotation="0,0,0"
    Scale="0,0,0"
   ) {scaleMode=BMM_Replace, alphaBoolBlend {blendOption=Linear}, boneToModify {boneName=streamer_L_01}, scale=<0 0 0>}
   n3 = "Transform (Modify) Bone - Bone: streamer_R_01"(ComponentPose=@n2, Alpha=@n1, Translation="0,0,0", Rotation="0,0,0", Scale="0,0,0") {scaleMode=BMM_Replace, alphaBoolBlend {blendOption=Linear}, boneToModify {boneName=streamer_R_01}, scale=<0 0 0>}
   n4 = Anim Dynamics - Bone: googly_eye_L(ComponentPose=@n3, Alpha=1, ExternalForce="0,0,0") {
    bGravityOverrideInSimSpace=true
    simulationSpace=World
    alphaBoolBlend {blendOption=Linear}
    boundBone {boneName=googly_eye_L}
    physicsBodyDefinitions [{
      sphereCollisionRadius=0
      boundBone {boneName=googly_eye_L}
      boxExtents=<1 1 1>
      constraintSetup {angularLimitsMax=<0 0 0>, angularLimitsMin=<0 0 0>, angularTarget=<0 0 0>, linearAxesMax=<1 3 0>, linearAxesMin=<-1 -5 0>}
      localJointOffset=<0 0 0>
     }]
   }
   n5 = Anim Dynamics - Bone: googly_eye_R(ComponentPose=@n4, Alpha=1, ExternalForce="0,0,0") {
    bGravityOverrideInSimSpace=true
    simulationSpace=World
    alphaBoolBlend {blendOption=Linear}
    boundBone {boneName=googly_eye_R}
    physicsBodyDefinitions [{
      sphereCollisionRadius=0
      boundBone {boneName=googly_eye_R}
      boxExtents=<1 1 1>
      constraintSetup {angularLimitsMax=<0 0 0>, angularLimitsMin=<0 0 0>, angularTarget=<0 0 0>, linearAxesMax=<1 3 0>, linearAxesMin=<-1 -5 0>}
      localJointOffset=<0 0 0>
     }]
   }
   n6 = "Clamp (Float)"(Value="To Float (Boolean)"(InBool=Get DynamicsAlpha() -> (DynamicsAlpha:bool)), Min=0, Max=0.3)
   Output Pose(Result=Component To Local(ComponentPose=RigidBody(ComponentPose=@n5, Alpha=@n6) {alphaBoolBlend {blendOption=Linear}})) {layerGroup=None}
   Get MeshToCopy() -> (MeshToCopy:Skeletal Mesh Component Object Reference)
  ]}
 EventGraph {nodes [
   n1 = Event BlueprintUpdateAnimation()
   n2 = GetOwningActor()
   > n3 = Cast To FortPlayerParachute!(execute, Object=@n2) -> (then, CastFailed, AsFort Player Parachute:Fort Player Parachute Object Reference) {is=title K2Node_DynamicCast}
   n4 = IsGliderFullyDeployedAccordingToAnim!(self:Fort Player Parachute Object Reference=@n3) -> (ReturnValue:bool)
   > Set DynamicsAlpha(DynamicsAlpha:bool=@n4) -> (Output_Get:bool)
   > n5 = Branch(Condition=NOT Boolean(A=@n4))
   > Set StreamersAlpha(StreamersAlpha:double=1) -> (Output_Get:double)
   Set StreamersAlpha(execute=@n5.else, StreamersAlpha=FInterpTo(Current=Get StreamersAlpha(), Target=0, DeltaTime=@n1.DeltaTimeX, InterpSpeed=6))
   Event BlueprintInitializeAnimation()
   > n6 = Cast To FortPlayerParachute(Object=@n2)
   > Set MeshToCopy(MeshToCopy:Skeletal Mesh Component Object Reference=Get ParachuteMesh!(self:Fort Player Parachute Object Reference=@n6) -> (ParachuteMesh:Skeletal Mesh Component Budgeted Object Reference)) -> (Output_Get:Skeletal Mesh Component Object Reference)
   TryGetPawnOwner()
  ]}
]
variables [
 DynamicsAlpha {type=boolean}
 MeshToCopy {type=SkeletalMeshComponent}
 StreamersAlpha {type=number, default=1}
]
