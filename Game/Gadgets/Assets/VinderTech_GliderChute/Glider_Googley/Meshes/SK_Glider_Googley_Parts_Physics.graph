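# Physics asset for the Googley glider parts mesh, covering its left and right streamers.
# Each streamer is a nine-bone chain of small capsules, with each bone constrained to the one before it.
# Every joint sets a modest Swing2 limit and zero Swing1 and twist limits, suggesting bending in one plane.
# The first left body also carries a larger offset capsule, and neither chain root has a constraint to a parent.
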PhysicsAsset /Game/Gadgets/Assets/VinderTech_GliderChute/Glider_Googley/Meshes/SK_Glider_Googley_Parts_Physics
bodies {
 streamer_L_01 [
  {shapeType=Capsule, length=0.5, radius=5}
  {shapeType=Capsule, length=15, radius=19.5, center=<-3.3923 -107.5564 -22.3634>, rotation=<rot 3.0156 -7.1228 -67.1684>}
 ]
 streamer_L_02 [{shapeType=Capsule, length=0.5, radius=5}]
 streamer_L_03 [{shapeType=Capsule, length=0.5, radius=5}]
 streamer_L_04 [{shapeType=Capsule, length=0.5, radius=5}]
 streamer_L_05 [{shapeType=Capsule, length=0.5, radius=5}]
 streamer_L_06 [{shapeType=Capsule, length=0.5, radius=5}]
 streamer_L_07 [{shapeType=Capsule, length=0.5, radius=5}]
 streamer_L_08 [{shapeType=Capsule, length=0.5, radius=5}]
 streamer_L_09 [{shapeType=Capsule, length=0.5, radius=5}]
 streamer_R_01 [{shapeType=Capsule, length=0.5, radius=5}]
 streamer_R_02 [{shapeType=Capsule, length=0.5, radius=5}]
 streamer_R_03 [{shapeType=Capsule, length=0.5, radius=5}]
 streamer_R_04 [{shapeType=Capsule, length=0.5, radius=5}]
 streamer_R_05 [{shapeType=Capsule, length=0.5, radius=5}]
 streamer_R_06 [{shapeType=Capsule, length=0.5, radius=5}]
 streamer_R_07 [{shapeType=Capsule, length=0.5, radius=5}]
 streamer_R_08 [{shapeType=Capsule, length=0.5, radius=5}]
 streamer_R_09 [{shapeType=Capsule, length=0.5, radius=5}]
}
constraints [
 {bone1Name=streamer_L_02, bone2Name=streamer_L_01, swing1LimitDegrees=0, swing2LimitDegrees=22.5, twistLimitDegrees=0}
 {bone1Name=streamer_L_03, bone2Name=streamer_L_02, swing1LimitDegrees=0, swing2LimitDegrees=22.5, twistLimitDegrees=0}
 {bone1Name=streamer_L_04, bone2Name=streamer_L_03, swing1LimitDegrees=0, swing2LimitDegrees=22.5, twistLimitDegrees=0}
 {bone1Name=streamer_L_05, bone2Name=streamer_L_04, swing1LimitDegrees=0, swing2LimitDegrees=22.5, twistLimitDegrees=0}
 {bone1Name=streamer_L_06, bone2Name=streamer_L_05, swing1LimitDegrees=0, swing2LimitDegrees=22.5, twistLimitDegrees=0}
 {bone1Name=streamer_L_07, bone2Name=streamer_L_06, swing1LimitDegrees=0, swing2LimitDegrees=22.5, twistLimitDegrees=0}
 {bone1Name=streamer_L_08, bone2Name=streamer_L_07, swing1LimitDegrees=0, swing2LimitDegrees=22.5, twistLimitDegrees=0}
 {bone1Name=streamer_L_09, bone2Name=streamer_L_08, swing1LimitDegrees=0, swing2LimitDegrees=22.5, twistLimitDegrees=0}
 {bone1Name=streamer_R_02, bone2Name=streamer_R_01, swing1LimitDegrees=0, swing2LimitDegrees=22.5, twistLimitDegrees=0}
 {bone1Name=streamer_R_03, bone2Name=streamer_R_02, swing1LimitDegrees=0, swing2LimitDegrees=22.5, twistLimitDegrees=0}
 {bone1Name=streamer_R_04, bone2Name=streamer_R_03, swing1LimitDegrees=0, swing2LimitDegrees=22.5, twistLimitDegrees=0}
 {bone1Name=streamer_R_05, bone2Name=streamer_R_04, swing1LimitDegrees=0, swing2LimitDegrees=22.5, twistLimitDegrees=0}
 {bone1Name=streamer_R_06, bone2Name=streamer_R_05, swing1LimitDegrees=0, swing2LimitDegrees=22.5, twistLimitDegrees=0}
 {bone1Name=streamer_R_07, bone2Name=streamer_R_06, swing1LimitDegrees=0, swing2LimitDegrees=22.5, twistLimitDegrees=0}
 {bone1Name=streamer_R_08, bone2Name=streamer_R_07, swing1LimitDegrees=0, swing2LimitDegrees=22.5, twistLimitDegrees=0}
 {bone1Name=streamer_R_09, bone2Name=streamer_R_08, swing1LimitDegrees=0, swing2LimitDegrees=22.5, twistLimitDegrees=0}
]
